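; GCN_RezOut_Realm is a burst gameplay cue notify that configures the parent cue's rez-out elimination animation and dissolve for Realm player pawns.
; On burst, it picks the with-weapon or no-weapon montage from the pawn's holstered-on-death flag.
; It disables the delay between animation and VFX when the pawn is falling or surface swimming.
; It then calls the parent burst, which is skipped if the pawn cast fails.
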(event EventBeginPlay)

(event Collision|EventActorBeginOverlap
  (params (OtherActor :type Actor%20Object%20Reference)))

(event EventTick
  (params (DeltaSeconds :type Float%20%28single-precision%29)))

(event EventOnBurst(EitherVFX-preferthisversion)
  (params (MyTarget :type Actor%20Object%20Reference) (Parameters :type Gameplay%20Cue%20Parameters%20Structure%20%28by%20ref%29) (ParticleComponents :type Array%20of%20FXSystem%20Component%20Object%20References) (AudioComponents :type Array%20of%20Audio%20Component%20Object%20References) (BurstCameraShakeInstance :type Camera%20Shake%20Base%20Object%20Reference) (BurstDecalInstance :type Decal%20Actor%20Object%20Reference))
  (bind _asbp_player_pawn_realm (Utilities|Casting|CastToBP_PlayerPawn_Realm MyTarget)
    (:then
      (if (Variables|/Realm/Pawn/BP_PlayerPawn_Realm.BP_PlayerPawn_Realm_C|GetWeaponHolsteredOnDeath _asbp_player_pawn_realm)
        (Variables|ELIMINATIONANIMS|SetBase%20Elimination%20Montage (Variables|Default|GetNoWeaponAnimMontage))
        (else
          (Variables|ELIMINATIONANIMS|SetBase%20Elimination%20Montage (Variables|Default|GetWithWeaponAnimMontage))))
      (if (or (AI|Components|NavMovement|IsFalling (Variables|/Script/Engine.Character|GetCharacterMovement _asbp_player_pawn_realm)) (WaterBody|IsSurfaceSwimming _asbp_player_pawn_realm))
        (Variables|DISSOLVE|SetShouldDelayBetweenAnimAndVFX false)
        (else
          (Variables|DISSOLVE|SetShouldDelayBetweenAnimAndVFX true)))
      (|CallParentFunction MyTarget Parameters ParticleComponents AudioComponents BurstCameraShakeInstance BurstDecalInstance))
    (:CastFailed
      _)))

(fn ConstructionScript
  (|CallParentFunction))
